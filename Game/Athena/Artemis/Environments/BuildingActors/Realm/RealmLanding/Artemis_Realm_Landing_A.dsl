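; Artemis realm landing building actor blueprint for Athena environments, with an optional landing beam effect.
; On BeginPlay it destroys its persistent FX component when the UseLandingBeam flag is off.
; The construction script only calls its parent.
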(event EventBeginPlay
  (if (not (Variables|Default|GetUseLandingBeam))
    (Components|DestroyComponent (Variables|Default|GetPersistentFXComponent))))

(fn ConstructionScript
  (|CallParentFunction))
