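; Realm landing beam actor that shows a cosmetic beacon, fading it in at BeginPlay and out after a 30-second timer.
; On clients able to run cosmetics, it creates dynamic materials for the flat and tall planes, caches their authored values, then zeroes visibility.
; FadeIn and FadeOut timelines lerp tall-plane opacity and gradient power plus flat-plane color alpha, each stopping the other timeline.
; FadeOut hides only the tall plane when finished.
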(event EventBeginPlay
  (bind (_flatstartingcolor_r _flatstartingcolor_g _flatstartingcolor_b) (Math|Color|BreakColor (Variables|GetflatStartingColor_R)))
  (|CallParentFunction)
  (Networking|CanExecuteCosmeticEvents
    (:True
      (Utilities|FlowControl|Sequence
        (:then_0
          (bind _returnvalue (Rendering|Material|CreateDynamicMaterialInstance (Variables|Default|GetFlat%20Plane) 0 0 "None" :!class /Script/Engine.PrimitiveComponent))
          (bind _output_get (Variables|Default|SetBeaconFlat _returnvalue))
          (bind _returnvalue_2 (Rendering|Material|GetVectorParameterValue _output_get "Color" :!class /Script/Engine.MaterialInstanceDynamic))
          (Variables|Default|SetflatStartingColor _returnvalue_2)
          (bind _returnvalue_3 (Rendering|Material|CreateDynamicMaterialInstance (Variables|Default|GetTallPlane) 0 0 "None" :!class /Script/Engine.PrimitiveComponent))
          (bind _output_get_2 (Variables|Default|SetBeaconTall _returnvalue_3))
          (bind _returnvalue_4 (Rendering|Material|GetScalarParameterValue _output_get_2 "Opacity" :!class /Script/Engine.MaterialInstanceDynamic))
          (Variables|Default|SetTallStartingOpacity _returnvalue_4)
          (bind _returnvalue_5 (Rendering|Material|GetScalarParameterValue (Variables|Default|GetBeaconTall) "GradientFadePower" :!class /Script/Engine.MaterialInstanceDynamic))
          (Variables|Default|SetTallStartingGradientPower _returnvalue_5)
          (Rendering|Material|SetScalarParameterValue (Variables|Default|GetBeaconTall) "Opacity" 0.0 :!class /Script/Engine.MaterialInstanceDynamic)
          (Rendering|Material|SetVectorParameterValue (Variables|Default|GetBeaconFlat) "Color" (Math|Color|MakeColor _flatstartingcolor_r _flatstartingcolor_g _flatstartingcolor_b 0.0) :!class /Script/Engine.MaterialInstanceDynamic)
          (Rendering|Material|SetScalarParameterValue (Variables|Default|GetBeaconTall) "GradientFadePower" 18.0 :!class /Script/Engine.MaterialInstanceDynamic))
        (:then_1
          (CallFunction|FadeIn))
        (:then_2
          (Utilities|Time|SetTimerbyEvent (delegate Custom|CustomEvent) 3e+01))))))

(event Custom|CustomEvent
  (CallFunction|FadeOut))

(event Custom|FadeIn
  (bind (_flatstartingcolor_r _flatstartingcolor_g _flatstartingcolor_b _flatstartingcolor_a) (Math|Color|BreakColor (Variables|GetflatStartingColor_R)))
  (Components|Timeline|Stop (Variables|Artemis_Realm_Landing_Beam|GetFadeOutTimeline))
  (Rendering|SetVisibility (Variables|Default|GetTallPlane) true false)
  (bind (_direction _alpha) (Timeline|FadeInTimeline PlayFromStart)
    (:Update
      (Rendering|Material|SetScalarParameterValue (Variables|Default|GetBeaconTall) "Opacity" (Math|Float|Lerp 0.0 (Variables|Default|GetTallStartingOpacity) _alpha) :!class /Script/Engine.MaterialInstanceDynamic)
      (Rendering|Material|SetScalarParameterValue (Variables|Default|GetBeaconTall) "GradientFadePower" (Math|Float|Lerp 18.0 (Variables|Default|GetTallStartingGradientPower) _alpha) :!class /Script/Engine.MaterialInstanceDynamic)
      (Rendering|Material|SetVectorParameterValue (Variables|Default|GetBeaconFlat) "Color" (Math|Color|MakeColor _flatstartingcolor_r _flatstartingcolor_g _flatstartingcolor_b (Math|Float|Lerp 0.0 _flatstartingcolor_a _alpha)) :!class /Script/Engine.MaterialInstanceDynamic))))

(event Custom|FadeOut
  (bind (_flatstartingcolor_r _flatstartingcolor_g _flatstartingcolor_b _flatstartingcolor_a) (Math|Color|BreakColor (Variables|GetflatStartingColor_R)))
  (Components|Timeline|Stop (Variables|Artemis_Realm_Landing_Beam|GetFadeInTimeline))
  (bind (_direction _alpha) (Timeline|FadeOutTimeline PlayFromStart)
    (:Update
      (Rendering|Material|SetScalarParameterValue (Variables|Default|GetBeaconTall) "Opacity" (Math|Float|Lerp 0.0 (Variables|Default|GetTallStartingOpacity) _alpha) :!class /Script/Engine.MaterialInstanceDynamic)
      (Rendering|Material|SetVectorParameterValue (Variables|Default|GetBeaconFlat) "Color" (Math|Color|MakeColor _flatstartingcolor_r _flatstartingcolor_g _flatstartingcolor_b (Math|Float|Lerp 0.0 _flatstartingcolor_a _alpha)) :!class /Script/Engine.MaterialInstanceDynamic)
      (Rendering|Material|SetScalarParameterValue (Variables|Default|GetBeaconTall) "GradientFadePower" (Math|Float|Lerp 18.0 (Variables|Default|GetTallStartingGradientPower) _alpha) :!class /Script/Engine.MaterialInstanceDynamic))
    (:Finished
      (Rendering|SetVisibility (Variables|Default|GetTallPlane) false false))))

(fn ConstructionScript
  (|CallParentFunction))
